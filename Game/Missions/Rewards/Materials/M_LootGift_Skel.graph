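# Unlit additive material for the skeletal LootGift mission reward mesh, flagged for mesh deformer use.
# A Fresnel term blends black into the Color parameter for a rim-glow term.
# Separate branches build a time-driven sine offset along Z, suggesting a bob, and a floored U-coordinate step mask.
# The graph shows no connections from these three branches to material outputs.
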Material /Game/Missions/Rewards/Materials/M_LootGift_Skel
graph {
 properties {
  bUsedWithMeshDeformer=true
  bUsedWithSkeletalMesh=true
  blendMode=BLEND_Additive
  pixelDepthOffsetMode=PDOM_Legacy
  previewMesh=/Game/Missions/Rewards/Meshes/S_LootGift
  shadingModel=MSM_Unlit
  translucencyLightingMode=TLM_VolumetricDirectional
  usedShadingModels=Unlit
 }
 nodes [
  n1 = TextureCoordinate()
  Floor(None=Multiply(A=ComponentMask(None=@n1) {r=true}) {constB=2})
  LinearInterpolate(A=Constant(), B=VectorParameter() {parameterName=Color, defaultValue=<rgba 0.445 0.3843 0.3649 1>}.RGB, Alpha=Fresnel() {exponent=1})
  Multiply(
   A=Constant3Vector() {constant=<rgba 0 0 1 0>}
   B=MaterialFunctionCall!(Sine Phase=Time(), Value 1=Constant() {r=8}, Value 2=Constant() {r=-8}) {materialFunction=/Engine/Functions/Engine_MaterialFunctions02/Utility/Sine_Remapped}.Remapped Sine
  )
 ]
}
